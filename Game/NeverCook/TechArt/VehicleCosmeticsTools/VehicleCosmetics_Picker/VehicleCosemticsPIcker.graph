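# Editor utility widget titled Vehicle Cosmetics Picker, a tech art tool for choosing vehicle body, skin and wheel cosmetics.
# PreConstruct points three DetailsViews at the widget itself, each filtered to expose one VCID variable.
# It also stores the skins DataTable's row names and spawns a transient actor from the body item at the origin.
# Paint and finish sections and the boost and jump animation buttons are layout only, with no graph logic.
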EditorUtilityWidgetBlueprint /Game/NeverCook/TechArt/VehicleCosmeticsTools/VehicleCosmetics_Picker/VehicleCosemticsPIcker
graphs [EventGraph {nodes [
   Event PreConstruct()
   > SetObject(self=Get DetailsView_0() -> (DetailsView_0:Details View Object Reference), NewObject=Self-Reference())
   > SetObject(self=Get DetailsView_2() -> (DetailsView_2:Details View Object Reference), NewObject=Self-Reference())
   > SetObject(self=Get DetailsView_60() -> (DetailsView_60:Details View Object Reference), NewObject=Self-Reference())
   > n1 = GetDataTableRowNames(Table=Get VCID_Skins() -> (VCID_Skins:Data Table Object Reference))
   > Set Out Row Names(Out Row Names:Array of Names=@n1) -> (Output_Get:Array of Names)
   > SpawnActorFromObject!(
    execute
    ObjectToUse:Object Reference=Get VCID_Body() -> (VCID_Body:Fort Item Definition Object Reference)
    Location:Vector="0, 0, 0"
    Rotation:Rotator
    bTransient:bool=true
   ) -> (then, ReturnValue:Actor Object Reference)
  ]}]
variables [
 Out Row Names {type="array<string>", default []}
 VCID_Body {type=FortItemDefinition, default=null}
 VCID_Skins {type=DataTable, default=null}
 VCID_Wheels {type=FortItemDefinition, default=null}
]
widgets {
 info {rootWidgetClass=&Overlay, widgetCount=57}
 widgets [
  Overlay_0 {widgetClassPath=&Overlay, widget=Overlay_0}
  ThresholdScrollBox_34 {
   widgetClassPath=/Script/FortniteUI.ThresholdScrollBox
   parent=Overlay_0
   widget=ThresholdScrollBox_34
   properties {
    widgetBarStyle {
     thickness=8
     draggedThumbImage {
      bIsDynamicallyLoaded=false
      drawAs=RoundedBox
      imageType=NoImage
      mirroring=NoMirror
      resourceName=None
      resourceObject=null
      tiling=NoTile
      imageSize=<0 0>
      margin=<ltrb 0 0 0 0>
      outlineSettings {
       bUseBrushTransparency=false
       roundingType=FixedRadius
       width=0
       color {colorUseRule=UseColor_Specified, specifiedColor=<rgba 0 0 0 0>}
       cornerRadii=<4 4 4 4>
      }
      tintColor {colorUseRule=UseColor_Specified, specifiedColor=<rgba 0.2159 0.2159 0.2159 1>}
      uVRegion {bIsValid=false, max=<0 0>, min=<0 0>}
     }
     horizontalBackgroundImage {
      bIsDynamicallyLoaded=false
      drawAs=NoDrawType
      imageType=NoImage
      mirroring=NoMirror
      resourceName=None
      resourceObject=null
      tiling=NoTile
      imageSize=<0 0>
      margin=<ltrb 0 0 0 0>
      outlineSettings {
       bUseBrushTransparency=false
       roundingType=HalfHeightRadius
       width=0
       color {colorUseRule=UseColor_Specified, specifiedColor=<rgba 0 0 0 0>}
       cornerRadii=<0 0 0 0>
      }
      tintColor {colorUseRule=UseColor_Specified, specifiedColor=<rgba 1 1 1 1>}
      uVRegion {bIsValid=false, max=<0 0>, min=<0 0>}
     }
     horizontalBottomSlotImage {
      bIsDynamicallyLoaded=false
      drawAs=NoDrawType
      imageType=NoImage
      mirroring=NoMirror
      resourceName=None
      resourceObject=null
      tiling=NoTile
      imageSize=<0 0>
      margin=<ltrb 0 0 0 0>
      outlineSettings {
       bUseBrushTransparency=false
       roundingType=HalfHeightRadius
       width=0
       color {colorUseRule=UseColor_Specified, specifiedColor=<rgba 0 0 0 0>}
       cornerRadii=<0 0 0 0>
      }
      tintColor {colorUseRule=UseColor_Specified, specifiedColor=<rgba 1 1 1 1>}
      uVRegion {bIsValid=false, max=<0 0>, min=<0 0>}
     }
     horizontalTopSlotImage {
      bIsDynamicallyLoaded=false
      drawAs=NoDrawType
      imageType=NoImage
      mirroring=NoMirror
      resourceName=None
      resourceObject=null
      tiling=NoTile
      imageSize=<0 0>
      margin=<ltrb 0 0 0 0>
      outlineSettings {
       bUseBrushTransparency=false
       roundingType=HalfHeightRadius
       width=0
       color {colorUseRule=UseColor_Specified, specifiedColor=<rgba 0 0 0 0>}
       cornerRadii=<0 0 0 0>
      }
      tintColor {colorUseRule=UseColor_Specified, specifiedColor=<rgba 1 1 1 1>}
      uVRegion {bIsValid=false, max=<0 0>, min=<0 0>}
     }
     hoveredThumbImage {
      bIsDynamicallyLoaded=false
      drawAs=RoundedBox
      imageType=NoImage
      mirroring=NoMirror
      resourceName=None
      resourceObject=null
      tiling=NoTile
      imageSize=<0 0>
      margin=<ltrb 0 0 0 0>
      outlineSettings {
       bUseBrushTransparency=false
       roundingType=FixedRadius
       width=0
       color {colorUseRule=UseColor_Specified, specifiedColor=<rgba 0 0 0 0>}
       cornerRadii=<4 4 4 4>
      }
      tintColor {colorUseRule=UseColor_Specified, specifiedColor=<rgba 0.2159 0.2159 0.2159 1>}
      uVRegion {bIsValid=false, max=<0 0>, min=<0 0>}
     }
     normalThumbImage {
      bIsDynamicallyLoaded=false
      drawAs=RoundedBox
      imageType=NoImage
      mirroring=NoMirror
      resourceName=None
      resourceObject=null
      tiling=NoTile
      imageSize=<0 0>
      margin=<ltrb 0 0 0 0>
      outlineSettings {
       bUseBrushTransparency=false
       roundingType=FixedRadius
       width=0
       color {colorUseRule=UseColor_Specified, specifiedColor=<rgba 0 0 0 0>}
       cornerRadii=<4 4 4 4>
      }
      tintColor {colorUseRule=UseColor_Specified, specifiedColor=<rgba 0.0953 0.0953 0.0953 1>}
      uVRegion {bIsValid=false, max=<0 0>, min=<0 0>}
     }
     verticalBackgroundImage {
      bIsDynamicallyLoaded=false
      drawAs=NoDrawType
      imageType=NoImage
      mirroring=NoMirror
      resourceName=None
      resourceObject=null
      tiling=NoTile
      imageSize=<0 0>
      margin=<ltrb 0 0 0 0>
      outlineSettings {
       bUseBrushTransparency=false
       roundingType=HalfHeightRadius
       width=0
       color {colorUseRule=UseColor_Specified, specifiedColor=<rgba 0 0 0 0>}
       cornerRadii=<0 0 0 0>
      }
      tintColor {colorUseRule=UseColor_Specified, specifiedColor=<rgba 1 1 1 1>}
      uVRegion {bIsValid=false, max=<0 0>, min=<0 0>}
     }
     verticalBottomSlotImage {
      bIsDynamicallyLoaded=false
      drawAs=NoDrawType
      imageType=NoImage
      mirroring=NoMirror
      resourceName=None
      resourceObject=null
      tiling=NoTile
      imageSize=<0 0>
      margin=<ltrb 0 0 0 0>
      outlineSettings {
       bUseBrushTransparency=false
       roundingType=HalfHeightRadius
       width=0
       color {colorUseRule=UseColor_Specified, specifiedColor=<rgba 0 0 0 0>}
       cornerRadii=<0 0 0 0>
      }
      tintColor {colorUseRule=UseColor_Specified, specifiedColor=<rgba 1 1 1 1>}
      uVRegion {bIsValid=false, max=<0 0>, min=<0 0>}
     }
     verticalTopSlotImage {
      bIsDynamicallyLoaded=false
      drawAs=NoDrawType
      imageType=NoImage
      mirroring=NoMirror
      resourceName=None
      resourceObject=null
      tiling=NoTile
      imageSize=<0 0>
      margin=<ltrb 0 0 0 0>
      outlineSettings {
       bUseBrushTransparency=false
       roundingType=HalfHeightRadius
       width=0
       color {colorUseRule=UseColor_Specified, specifiedColor=<rgba 0 0 0 0>}
       cornerRadii=<0 0 0 0>
      }
      tintColor {colorUseRule=UseColor_Specified, specifiedColor=<rgba 1 1 1 1>}
      uVRegion {bIsValid=false, max=<0 0>, min=<0 0>}
     }
    }
    widgetStyle {
     barThickness=8
     bottomShadowBrush {
      bIsDynamicallyLoaded=false
      drawAs=Box
      imageType=FullColor
      mirroring=NoMirror
      resourceName=../../../Engine/Content/Slate/Common/ScrollBoxShadowBottom.png
      resourceObject=null
      tiling=NoTile
      imageSize=<16 8>
      margin=<ltrb 0.5 0 0.5 1>
      outlineSettings {
       bUseBrushTransparency=false
       roundingType=HalfHeightRadius
       width=0
       color {colorUseRule=UseColor_Specified, specifiedColor=<rgba 0 0 0 0>}
       cornerRadii=<0 0 0 0>
      }
      tintColor {colorUseRule=UseColor_Specified, specifiedColor=<rgba 1 1 1 1>}
      uVRegion {bIsValid=false, max=<0 0>, min=<0 0>}
     }
     horizontalScrolledContentPadding=<ltrb 0 0 1 0>
     leftShadowBrush {
      bIsDynamicallyLoaded=false
      drawAs=Box
      imageType=FullColor
      mirroring=NoMirror
      resourceName=../../../Engine/Content/Slate/Common/ScrollBoxShadowLeft.png
      resourceObject=null
      tiling=NoTile
      imageSize=<8 16>
      margin=<ltrb 1 0.5 0 0.5>
      outlineSettings {
       bUseBrushTransparency=false
       roundingType=HalfHeightRadius
       width=0
       color {colorUseRule=UseColor_Specified, specifiedColor=<rgba 0 0 0 0>}
       cornerRadii=<0 0 0 0>
      }
      tintColor {colorUseRule=UseColor_Specified, specifiedColor=<rgba 1 1 1 1>}
      uVRegion {bIsValid=false, max=<0 0>, min=<0 0>}
     }
     rightShadowBrush {
      bIsDynamicallyLoaded=false
      drawAs=Box
      imageType=FullColor
      mirroring=NoMirror
      resourceName=../../../Engine/Content/Slate/Common/ScrollBoxShadowRight.png
      resourceObject=null
      tiling=NoTile
      imageSize=<8 16>
      margin=<ltrb 0 0.5 1 0.5>
      outlineSettings {
       bUseBrushTransparency=false
       roundingType=HalfHeightRadius
       width=0
       color {colorUseRule=UseColor_Specified, specifiedColor=<rgba 0 0 0 0>}
       cornerRadii=<0 0 0 0>
      }
      tintColor {colorUseRule=UseColor_Specified, specifiedColor=<rgba 1 1 1 1>}
      uVRegion {bIsValid=false, max=<0 0>, min=<0 0>}
     }
     topShadowBrush {
      bIsDynamicallyLoaded=false
      drawAs=Box
      imageType=FullColor
      mirroring=NoMirror
      resourceName=../../../Engine/Content/Slate/Common/ScrollBoxShadowTop.png
      resourceObject=null
      tiling=NoTile
      imageSize=<16 8>
      margin=<ltrb 0.5 1 0.5 0>
      outlineSettings {
       bUseBrushTransparency=false
       roundingType=HalfHeightRadius
       width=0
       color {colorUseRule=UseColor_Specified, specifiedColor=<rgba 0 0 0 0>}
       cornerRadii=<0 0 0 0>
      }
      tintColor {colorUseRule=UseColor_Specified, specifiedColor=<rgba 1 1 1 1>}
      uVRegion {bIsValid=false, max=<0 0>, min=<0 0>}
     }
     verticalScrolledContentPadding=<ltrb 0 0 0 1>
    }
   }
   slot {class=OverlaySlot, properties {horizontalAlignment=HAlign_Fill, verticalAlignment=VAlign_Fill}}
  }
  VerticalBox_35 {
   widgetClassPath=&VerticalBox
   parent=ThresholdScrollBox_34
   widget=VerticalBox_35
   slot {class=ScrollBoxSlot, properties {padding=<ltrb 0 25 0 0>, size {sizeRule=Fill, value=1}}}
  }
  ScaleBox_0 {
   widgetClassPath=&ScaleBox
   parent=VerticalBox_35
   widget=ScaleBox_0
   slot {class=VerticalBoxSlot, properties {horizontalAlignment=HAlign_Left, padding=<ltrb 16 16 16 16>}}
  }
  VerticalBox_TopLogo {widgetClassPath=&VerticalBox, parent=ScaleBox_0, widget=VerticalBox_TopLogo, slot {class=ScaleBoxSlot}}
  HorizontalBox_1 {widgetClassPath=&HorizontalBox, parent=VerticalBox_TopLogo, widget=HorizontalBox_1, slot {class=VerticalBoxSlot}}
  SizeBox_701 {widgetClassPath=&SizeBox, parent=HorizontalBox_1, widget=SizeBox_701, slot {class=HorizontalBoxSlot}}
  Image_35 {
   widgetClassPath=&Image
   bIsVariable=true
   parent=SizeBox_701
   widget=Image_35
   properties {brush {
     bIsDynamicallyLoaded=false
     drawAs=Image
     imageType=NoImage
     mirroring=NoMirror
     resourceName=None
     resourceObject=/Game/NeverCook/TechArt/VehicleCosmeticsTools/VehicleCosmetics_Picker/Icons/UI_RL_Icon
     tiling=NoTile
     imageSize=<64 64>
     margin=<ltrb 0 0 0 0>
     outlineSettings {
      bUseBrushTransparency=false
      roundingType=HalfHeightRadius
      width=0
      color {colorUseRule=UseColor_Specified, specifiedColor=<rgba 0 0 0 0>}
      cornerRadii=<0 0 0 0>
     }
     tintColor {colorUseRule=UseColor_Specified, specifiedColor=<rgba 1 1 1 1>}
     uVRegion {bIsValid=false, max=<0 0>, min=<0 0>}
    }}
   slot {class=SizeBoxSlot, properties {horizontalAlignment=HAlign_Left, verticalAlignment=VAlign_Center}}
  }
  SizeBox_754 {widgetClassPath=&SizeBox, parent=HorizontalBox_1, widget=SizeBox_754, slot {class=HorizontalBoxSlot}}
  TextBlock_310 {
   widgetClassPath=&TextBlock
   parent=SizeBox_754
   widget=TextBlock_310
   properties {
    text=Vehicle Cosmetics Picker
    font {
     bForceMonospaced=false
     bMaterialIsStencil=false
     fontMaterial=null
     fontObject=null
     letterSpacing=0
     monospacedWidth=1
     size=24
     skewAmount=0
     typefaceFontName=Default
     outlineSettings {
      bApplyOutlineToDropShadows=false
      bMiteredCorners=false
      bSeparateFillAlpha=false
      outlineMaterial=null
      outlineSize=0
      outlineColor=<rgba 0 0 0 1>
     }
    }
   }
   slot {class=SizeBoxSlot, properties {padding=<ltrb 16 16 16 16>}}
  }
  Border_90 {
   widgetClassPath=&Border
   parent=VerticalBox_TopLogo
   widget=Border_90
   properties {brushColor=<rgba 0.0833 0.0833 0.0833 1>}
   slot {class=VerticalBoxSlot, properties {padding=<ltrb 0 16 0 0>}}
  }
  VerticalBox_Body {widgetClassPath=&VerticalBox, parent=VerticalBox_35, widget=VerticalBox_Body, slot {class=VerticalBoxSlot}}
  SizeBox_BodyText {
   widgetClassPath=&SizeBox
   parent=VerticalBox_Body
   widget=SizeBox_BodyText
   slot {class=VerticalBoxSlot, properties {padding=<ltrb 16 45 16 8>}}
  }
  TextBlock {
   widgetClassPath=&TextBlock
   parent=SizeBox_BodyText
   widget=TextBlock
   properties {
    text=BODY
    colorAndOpacity {colorUseRule=UseColor_Specified, specifiedColor=<rgba 0.526 0.526 0.526 1>}
    font {
     bForceMonospaced=false
     bMaterialIsStencil=false
     fontMaterial=null
     fontObject=null
     letterSpacing=0
     monospacedWidth=1
     size=18
     skewAmount=0
     typefaceFontName=RLSmall-Bold
     outlineSettings {
      bApplyOutlineToDropShadows=false
      bMiteredCorners=false
      bSeparateFillAlpha=false
      outlineMaterial=null
      outlineSize=0
      outlineColor=<rgba 0 0 0 1>
     }
    }
   }
   slot {class=SizeBoxSlot, properties {horizontalAlignment=HAlign_Left}}
  }
  SizeBox_49 {
   widgetClassPath=&SizeBox
   parent=VerticalBox_Body
   widget=SizeBox_49
   slot {class=VerticalBoxSlot, properties {padding=<ltrb 16 0 16 16>}}
  }
  DetailsView_0 {
   widgetClassPath=&DetailsView
   parent=SizeBox_49
   widget=DetailsView_0
   properties {categoriesToShow [VCID_Body]}
   slot {class=SizeBoxSlot}
  }
  SizeBox_255 {
   widgetClassPath=&SizeBox
   parent=VerticalBox_Body
   widget=SizeBox_255
   slot {class=VerticalBoxSlot, properties {padding=<ltrb 16 0 16 8>}}
  }
  TextBlock_118 {
   widgetClassPath=&TextBlock
   parent=SizeBox_255
   widget=TextBlock_118
   properties {
    text=BODY PAINT
    colorAndOpacity {colorUseRule=UseColor_Specified, specifiedColor=<rgba 0.526 0.526 0.526 1>}
    font {
     bForceMonospaced=false
     bMaterialIsStencil=false
     fontMaterial=null
     fontObject=null
     letterSpacing=0
     monospacedWidth=1
     size=12
     skewAmount=0
     typefaceFontName=RLSmall-Italic
     outlineSettings {
      bApplyOutlineToDropShadows=false
      bMiteredCorners=false
      bSeparateFillAlpha=false
      outlineMaterial=null
      outlineSize=0
      outlineColor=<rgba 0 0 0 1>
     }
    }
   }
   slot {class=SizeBoxSlot, properties {horizontalAlignment=HAlign_Left}}
  }
  SizeBox {
   widgetClassPath=&SizeBox
   parent=VerticalBox_Body
   widget=SizeBox
   slot {class=VerticalBoxSlot, properties {padding=<ltrb 16 0 16 16>}}
  }
  DetailsView {widgetClassPath=&DetailsView, parent=SizeBox, widget=DetailsView, slot {class=SizeBoxSlot}}
  VerticalBox_Skins {widgetClassPath=&VerticalBox, parent=VerticalBox_35, widget=VerticalBox_Skins, slot {class=VerticalBoxSlot}}
  SizeBox_SkinsText {
   widgetClassPath=&SizeBox
   parent=VerticalBox_Skins
   widget=SizeBox_SkinsText
   slot {class=VerticalBoxSlot, properties {padding=<ltrb 16 45 16 8>}}
  }
  TextBlock_46 {
   widgetClassPath=&TextBlock
   parent=SizeBox_SkinsText
   widget=TextBlock_46
   properties {
    text=SKINS
    colorAndOpacity {colorUseRule=UseColor_Specified, specifiedColor=<rgba 0.526 0.526 0.526 1>}
    font {
     bForceMonospaced=false
     bMaterialIsStencil=false
     fontMaterial=null
     fontObject=null
     letterSpacing=0
     monospacedWidth=1
     size=18
     skewAmount=0
     typefaceFontName=RLSmall-Bold
     outlineSettings {
      bApplyOutlineToDropShadows=false
      bMiteredCorners=false
      bSeparateFillAlpha=false
      outlineMaterial=null
      outlineSize=0
      outlineColor=<rgba 0 0 0 1>
     }
    }
   }
   slot {class=SizeBoxSlot, properties {horizontalAlignment=HAlign_Left}}
  }
  SizeBox_SkinsDetailsView {widgetClassPath=&SizeBox, parent=VerticalBox_Skins, widget=SizeBox_SkinsDetailsView, slot {class=VerticalBoxSlot}}
  DetailsView_60 {
   widgetClassPath=&DetailsView
   parent=SizeBox_SkinsDetailsView
   widget=DetailsView_60
   properties {categoriesToShow [VCID_Skins]}
   slot {class=SizeBoxSlot, properties {padding=<ltrb 16 0 0 0>}}
  }
  HorizontalBox_Skins {
   widgetClassPath=&HorizontalBox
   parent=VerticalBox_Skins
   widget=HorizontalBox_Skins
   slot {class=VerticalBoxSlot, properties {padding=<ltrb 16 16 16 16>}}
  }
  VerticalBox_110 {
   widgetClassPath=&VerticalBox
   parent=HorizontalBox_Skins
   widget=VerticalBox_110
   slot {class=HorizontalBoxSlot, properties {padding=<ltrb 16 0 16 8>, size {sizeRule=Fill, value=1}}}
  }
  SizeBox_1 {widgetClassPath=&SizeBox, parent=VerticalBox_110, widget=SizeBox_1, slot {class=VerticalBoxSlot}}
  TextBlock_1 {
   widgetClassPath=&TextBlock
   parent=SizeBox_1
   widget=TextBlock_1
   properties {
    text=SKIN PAINT
    colorAndOpacity {colorUseRule=UseColor_Specified, specifiedColor=<rgba 0.526 0.526 0.526 1>}
    font {
     bForceMonospaced=false
     bMaterialIsStencil=false
     fontMaterial=null
     fontObject=null
     letterSpacing=0
     monospacedWidth=1
     size=12
     skewAmount=0
     typefaceFontName=RLSmall-Italic
     outlineSettings {
      bApplyOutlineToDropShadows=false
      bMiteredCorners=false
      bSeparateFillAlpha=false
      outlineMaterial=null
      outlineSize=0
      outlineColor=<rgba 0 0 0 1>
     }
    }
   }
   slot {class=SizeBoxSlot, properties {horizontalAlignment=HAlign_Left}}
  }
  SizeBox_2 {widgetClassPath=&SizeBox, parent=VerticalBox_110, widget=SizeBox_2, slot {class=VerticalBoxSlot}}
  DetailsView_1 {widgetClassPath=&DetailsView, parent=SizeBox_2, widget=DetailsView_1, slot {class=SizeBoxSlot}}
  VerticalBox_108 {
   widgetClassPath=&VerticalBox
   parent=HorizontalBox_Skins
   widget=VerticalBox_108
   slot {class=HorizontalBoxSlot, properties {size {sizeRule=Fill, value=1}}}
  }
  SizeBox_480 {widgetClassPath=&SizeBox, parent=VerticalBox_108, widget=SizeBox_480, slot {class=VerticalBoxSlot}}
  TextBlock_213 {
   widgetClassPath=&TextBlock
   parent=SizeBox_480
   widget=TextBlock_213
   properties {
    text=PAINT FINISH
    colorAndOpacity {colorUseRule=UseColor_Specified, specifiedColor=<rgba 0.526 0.526 0.526 1>}
    font {
     bForceMonospaced=false
     bMaterialIsStencil=false
     fontMaterial=null
     fontObject=null
     letterSpacing=0
     monospacedWidth=1
     size=12
     skewAmount=0
     typefaceFontName=RLSmall-Italic
     outlineSettings {
      bApplyOutlineToDropShadows=false
      bMiteredCorners=false
      bSeparateFillAlpha=false
      outlineMaterial=null
      outlineSize=0
      outlineColor=<rgba 0 0 0 1>
     }
    }
   }
   slot {class=SizeBoxSlot}
  }
  SizeBox_6 {
   widgetClassPath=&SizeBox
   parent=VerticalBox_108
   widget=SizeBox_6
   slot {class=VerticalBoxSlot, properties {horizontalAlignment=HAlign_Left}}
  }
  TextBlock_4 {
   widgetClassPath=&TextBlock
   parent=SizeBox_6
   widget=TextBlock_4
   properties {
    text=BASE
    colorAndOpacity {colorUseRule=UseColor_Specified, specifiedColor=<rgba 0.526 0.526 0.526 1>}
    font {
     bForceMonospaced=false
     bMaterialIsStencil=false
     fontMaterial=null
     fontObject=null
     letterSpacing=0
     monospacedWidth=1
     size=9
     skewAmount=0
     typefaceFontName=RLSmall-Italic
     outlineSettings {
      bApplyOutlineToDropShadows=false
      bMiteredCorners=false
      bSeparateFillAlpha=false
      outlineMaterial=null
      outlineSize=0
      outlineColor=<rgba 0 0 0 1>
     }
    }
   }
   slot {class=SizeBoxSlot, properties {horizontalAlignment=HAlign_Center}}
  }
  SizeBox_7 {
   widgetClassPath=&SizeBox
   parent=VerticalBox_108
   widget=SizeBox_7
   slot {class=VerticalBoxSlot, properties {horizontalAlignment=HAlign_Left}}
  }
  TextBlock_5 {
   widgetClassPath=&TextBlock
   parent=SizeBox_7
   widget=TextBlock_5
   properties {
    text=ACCENT
    colorAndOpacity {colorUseRule=UseColor_Specified, specifiedColor=<rgba 0.526 0.526 0.526 1>}
    font {
     bForceMonospaced=false
     bMaterialIsStencil=false
     fontMaterial=null
     fontObject=null
     letterSpacing=0
     monospacedWidth=1
     size=9
     skewAmount=0
     typefaceFontName=RLSmall-Italic
     outlineSettings {
      bApplyOutlineToDropShadows=false
      bMiteredCorners=false
      bSeparateFillAlpha=false
      outlineMaterial=null
      outlineSize=0
      outlineColor=<rgba 0 0 0 1>
     }
    }
   }
   slot {class=SizeBoxSlot, properties {horizontalAlignment=HAlign_Center}}
  }
  VerticalBox_Wheels {widgetClassPath=&VerticalBox, parent=VerticalBox_35, widget=VerticalBox_Wheels, slot {class=VerticalBoxSlot}}
  SizeBox_WheelsText {
   widgetClassPath=&SizeBox
   parent=VerticalBox_Wheels
   widget=SizeBox_WheelsText
   slot {class=VerticalBoxSlot, properties {padding=<ltrb 16 45 16 8>}}
  }
  TextBlock_2 {
   widgetClassPath=&TextBlock
   parent=SizeBox_WheelsText
   widget=TextBlock_2
   properties {
    text=WHEELS
    colorAndOpacity {colorUseRule=UseColor_Specified, specifiedColor=<rgba 0.526 0.526 0.526 1>}
    font {
     bForceMonospaced=false
     bMaterialIsStencil=false
     fontMaterial=null
     fontObject=null
     letterSpacing=0
     monospacedWidth=1
     size=18
     skewAmount=0
     typefaceFontName=RLSmall-Bold
     outlineSettings {
      bApplyOutlineToDropShadows=false
      bMiteredCorners=false
      bSeparateFillAlpha=false
      outlineMaterial=null
      outlineSize=0
      outlineColor=<rgba 0 0 0 1>
     }
    }
   }
   slot {class=SizeBoxSlot, properties {horizontalAlignment=HAlign_Left}}
  }
  SizeBox_3 {
   widgetClassPath=&SizeBox
   parent=VerticalBox_Wheels
   widget=SizeBox_3
   slot {class=VerticalBoxSlot, properties {padding=<ltrb 16 0 16 16>}}
  }
  DetailsView_2 {
   widgetClassPath=&DetailsView
   parent=SizeBox_3
   widget=DetailsView_2
   properties {categoriesToShow [VCID_Wheels]}
   slot {class=SizeBoxSlot}
  }
  SizeBox_4 {
   widgetClassPath=&SizeBox
   parent=VerticalBox_Wheels
   widget=SizeBox_4
   slot {class=VerticalBoxSlot, properties {padding=<ltrb 16 0 16 8>}}
  }
  TextBlock_3 {
   widgetClassPath=&TextBlock
   parent=SizeBox_4
   widget=TextBlock_3
   properties {
    text=WHEEL PAINT
    colorAndOpacity {colorUseRule=UseColor_Specified, specifiedColor=<rgba 0.526 0.526 0.526 1>}
    font {
     bForceMonospaced=false
     bMaterialIsStencil=false
     fontMaterial=null
     fontObject=null
     letterSpacing=0
     monospacedWidth=1
     size=12
     skewAmount=0
     typefaceFontName=RLSmall-Italic
     outlineSettings {
      bApplyOutlineToDropShadows=false
      bMiteredCorners=false
      bSeparateFillAlpha=false
      outlineMaterial=null
      outlineSize=0
      outlineColor=<rgba 0 0 0 1>
     }
    }
   }
   slot {class=SizeBoxSlot, properties {horizontalAlignment=HAlign_Left}}
  }
  SizeBox_5 {
   widgetClassPath=&SizeBox
   parent=VerticalBox_Wheels
   widget=SizeBox_5
   slot {class=VerticalBoxSlot, properties {padding=<ltrb 16 0 16 16>}}
  }
  DetailsView_3 {
   widgetClassPath=&DetailsView
   parent=SizeBox_5
   widget=DetailsView_3
   properties {categoriesToShow [None]}
   slot {class=SizeBoxSlot}
  }
  VerticalBox_Animation {widgetClassPath=&VerticalBox, parent=VerticalBox_35, widget=VerticalBox_Animation, slot {class=VerticalBoxSlot}}
  SizeBox_WheelsText_1 {
   widgetClassPath=&SizeBox
   parent=VerticalBox_Animation
   widget=SizeBox_WheelsText_1
   slot {class=VerticalBoxSlot, properties {padding=<ltrb 16 45 16 8>}}
  }
  TextBlock_6 {
   widgetClassPath=&TextBlock
   parent=SizeBox_WheelsText_1
   widget=TextBlock_6
   properties {
    text=ANIMATION
    colorAndOpacity {colorUseRule=UseColor_Specified, specifiedColor=<rgba 0.526 0.526 0.526 1>}
    font {
     bForceMonospaced=false
     bMaterialIsStencil=false
     fontMaterial=null
     fontObject=null
     letterSpacing=0
     monospacedWidth=1
     size=18
     skewAmount=0
     typefaceFontName=RLSmall-Bold
     outlineSettings {
      bApplyOutlineToDropShadows=false
      bMiteredCorners=false
      bSeparateFillAlpha=false
      outlineMaterial=null
      outlineSize=0
      outlineColor=<rgba 0 0 0 1>
     }
    }
   }
   slot {class=SizeBoxSlot, properties {horizontalAlignment=HAlign_Left}}
  }
  VerticalBox_112 {
   widgetClassPath=&VerticalBox
   parent=VerticalBox_Animation
   widget=VerticalBox_112
   slot {class=VerticalBoxSlot, properties {padding=<ltrb 16 0 16 8>}}
  }
  SizeBox_9 {widgetClassPath=&SizeBox, parent=VerticalBox_112, widget=SizeBox_9, slot {class=VerticalBoxSlot}}
  EditorUtilityButton_230 {
   widgetClassPath=&EditorUtilityButton
   parent=SizeBox_9
   widget=EditorUtilityButton_230
   properties {widgetStyle {
     clickedSlateSound {resourceObject=null}
     disabled {
      bIsDynamicallyLoaded=false
      drawAs=RoundedBox
      imageType=NoImage
      mirroring=NoMirror
      resourceName=None
      resourceObject=null
      tiling=NoTile
      imageSize=<0 0>
      margin=<ltrb 0 0 0 0>
      outlineSettings {
       bUseBrushTransparency=false
       roundingType=FixedRadius
       width=1
       color {colorUseRule=UseColor_ColorTable, specifiedColor=<rgba 0 0 0 0>}
       cornerRadii=<4 4 4 4>
      }
      tintColor {colorUseRule=UseColor_Specified, specifiedColor=<rgba 0.0395 0.0395 0.0395 1>}
      uVRegion {bIsValid=false, max=<0 0>, min=<0 0>}
     }
     disabledForeground {colorUseRule=UseColor_ColorTable, specifiedColor=<rgba 0 0 0 0>}
     hovered {
      bIsDynamicallyLoaded=false
      drawAs=RoundedBox
      imageType=NoImage
      mirroring=NoMirror
      resourceName=None
      resourceObject=null
      tiling=NoTile
      imageSize=<32 32>
      margin=<ltrb 0 0 0 0>
      outlineSettings {
       bUseBrushTransparency=false
       roundingType=FixedRadius
       width=1
       color {colorUseRule=UseColor_ColorTable, specifiedColor=<rgba 0 0 0 0>}
       cornerRadii=<4 4 4 4>
      }
      tintColor {colorUseRule=UseColor_Specified, specifiedColor=<rgba 0.7865 0.3344 0 1>}
      uVRegion {bIsValid=false, max=<0 0>, min=<0 0>}
     }
     hoveredForeground {colorUseRule=UseColor_ColorTable, specifiedColor=<rgba 0 0 0 0>}
     hoveredSlateSound {resourceObject=null}
     normal {
      bIsDynamicallyLoaded=false
      drawAs=RoundedBox
      imageType=NoImage
      mirroring=NoMirror
      resourceName=None
      resourceObject=null
      tiling=NoTile
      imageSize=<32 32>
      margin=<ltrb 0 0 0 0>
      outlineSettings {
       bUseBrushTransparency=false
       roundingType=FixedRadius
       width=1
       color {colorUseRule=UseColor_ColorTable, specifiedColor=<rgba 0 0 0 0>}
       cornerRadii=<4 4 4 4>
      }
      tintColor {colorUseRule=UseColor_Specified, specifiedColor=<rgba 0.3542 0.1506 0 1>}
      uVRegion {bIsValid=false, max=<0 0>, min=<0 0>}
     }
     normalForeground {colorUseRule=UseColor_ColorTable, specifiedColor=<rgba 0 0 0 0>}
     normalPadding=<ltrb 12 1.5 12 1.5>
     pressed {
      bIsDynamicallyLoaded=false
      drawAs=RoundedBox
      imageType=NoImage
      mirroring=NoMirror
      resourceName=None
      resourceObject=null
      tiling=NoTile
      imageSize=<32 32>
      margin=<ltrb 0 0 0 0>
      outlineSettings {
       bUseBrushTransparency=false
       roundingType=FixedRadius
       width=1
       color {colorUseRule=UseColor_ColorTable, specifiedColor=<rgba 0 0 0 0>}
       cornerRadii=<4 4 4 4>
      }
      tintColor {colorUseRule=UseColor_Specified, specifiedColor=<rgba 0.0885 0.0377 0 1>}
      uVRegion {bIsValid=false, max=<0 0>, min=<0 0>}
     }
     pressedForeground {colorUseRule=UseColor_ColorTable, specifiedColor=<rgba 0 0 0 0>}
     pressedPadding=<ltrb 12 2.5 12 0.5>
     pressedSlateSound {resourceObject=null}
    }}
   slot {class=SizeBoxSlot, properties {padding=<ltrb 0 12 0 0>}}
  }
  TextBlock_7 {
   widgetClassPath=&TextBlock
   parent=EditorUtilityButton_230
   widget=TextBlock_7
   properties {
    text=Play Boost Animation
    colorAndOpacity {colorUseRule=UseColor_Specified, specifiedColor=<rgba 0.526 0.526 0.526 1>}
    font {
     bForceMonospaced=false
     bMaterialIsStencil=false
     fontMaterial=null
     fontObject=null
     letterSpacing=0
     monospacedWidth=1
     size=12
     skewAmount=0
     typefaceFontName=RLSmall-Italic
     outlineSettings {
      bApplyOutlineToDropShadows=false
      bMiteredCorners=false
      bSeparateFillAlpha=false
      outlineMaterial=null
      outlineSize=0
      outlineColor=<rgba 0 0 0 1>
     }
    }
   }
   slot {class=ButtonSlot, properties {horizontalAlignment=HAlign_Left, verticalAlignment=VAlign_Fill, padding=<ltrb 0 0 0 0>}}
  }
  SizeBox_8 {
   widgetClassPath=&SizeBox
   parent=VerticalBox_112
   widget=SizeBox_8
   slot {class=VerticalBoxSlot, properties {padding=<ltrb 0 12 0 0>}}
  }
  EditorUtilityButton {
   widgetClassPath=&EditorUtilityButton
   parent=SizeBox_8
   widget=EditorUtilityButton
   properties {widgetStyle {
     clickedSlateSound {resourceObject=null}
     disabled {
      bIsDynamicallyLoaded=false
      drawAs=RoundedBox
      imageType=NoImage
      mirroring=NoMirror
      resourceName=None
      resourceObject=null
      tiling=NoTile
      imageSize=<0 0>
      margin=<ltrb 0 0 0 0>
      outlineSettings {
       bUseBrushTransparency=false
       roundingType=FixedRadius
       width=1
       color {colorUseRule=UseColor_ColorTable, specifiedColor=<rgba 0 0 0 0>}
       cornerRadii=<4 4 4 4>
      }
      tintColor {colorUseRule=UseColor_Specified, specifiedColor=<rgba 0.0395 0.0395 0.0395 1>}
      uVRegion {bIsValid=false, max=<0 0>, min=<0 0>}
     }
     disabledForeground {colorUseRule=UseColor_ColorTable, specifiedColor=<rgba 0 0 0 0>}
     hovered {
      bIsDynamicallyLoaded=false
      drawAs=RoundedBox
      imageType=NoImage
      mirroring=NoMirror
      resourceName=None
      resourceObject=null
      tiling=NoTile
      imageSize=<32 32>
      margin=<ltrb 0 0 0 0>
      outlineSettings {
       bUseBrushTransparency=false
       roundingType=FixedRadius
       width=1
       color {colorUseRule=UseColor_ColorTable, specifiedColor=<rgba 0 0 0 0>}
       cornerRadii=<4 4 4 4>
      }
      tintColor {colorUseRule=UseColor_Specified, specifiedColor=<rgba 0.3073 0.4066 1 1>}
      uVRegion {bIsValid=false, max=<0 0>, min=<0 0>}
     }
     hoveredForeground {colorUseRule=UseColor_ColorTable, specifiedColor=<rgba 0 0 0 0>}
     hoveredSlateSound {resourceObject=null}
     normal {
      bIsDynamicallyLoaded=false
      drawAs=RoundedBox
      imageType=NoImage
      mirroring=NoMirror
      resourceName=None
      resourceObject=null
      tiling=NoTile
      imageSize=<32 32>
      margin=<ltrb 0 0 0 0>
      outlineSettings {
       bUseBrushTransparency=false
       roundingType=FixedRadius
       width=1
       color {colorUseRule=UseColor_ColorTable, specifiedColor=<rgba 0 0 0 0>}
       cornerRadii=<4 4 4 4>
      }
      tintColor {colorUseRule=UseColor_Specified, specifiedColor=<rgba 0 0.1434 1 1>}
      uVRegion {bIsValid=false, max=<0 0>, min=<0 0>}
     }
     normalForeground {colorUseRule=UseColor_ColorTable, specifiedColor=<rgba 0 0 0 0>}
     normalPadding=<ltrb 12 1.5 12 1.5>
     pressed {
      bIsDynamicallyLoaded=false
      drawAs=RoundedBox
      imageType=NoImage
      mirroring=NoMirror
      resourceName=None
      resourceObject=null
      tiling=NoTile
      imageSize=<32 32>
      margin=<ltrb 0 0 0 0>
      outlineSettings {
       bUseBrushTransparency=false
       roundingType=FixedRadius
       width=1
       color {colorUseRule=UseColor_ColorTable, specifiedColor=<rgba 0 0 0 0>}
       cornerRadii=<4 4 4 4>
      }
      tintColor {colorUseRule=UseColor_Specified, specifiedColor=<rgba 0 0.0254 0.1771 1>}
      uVRegion {bIsValid=false, max=<0 0>, min=<0 0>}
     }
     pressedForeground {colorUseRule=UseColor_ColorTable, specifiedColor=<rgba 0 0 0 0>}
     pressedPadding=<ltrb 12 2.5 12 0.5>
     pressedSlateSound {resourceObject=null}
    }}
   slot {class=SizeBoxSlot}
  }
  TextBlock_8 {
   widgetClassPath=&TextBlock
   parent=EditorUtilityButton
   widget=TextBlock_8
   properties {
    text=Play Jump Animation
    colorAndOpacity {colorUseRule=UseColor_Specified, specifiedColor=<rgba 0.526 0.526 0.526 1>}
    font {
     bForceMonospaced=false
     bMaterialIsStencil=false
     fontMaterial=null
     fontObject=null
     letterSpacing=0
     monospacedWidth=1
     size=12
     skewAmount=0
     typefaceFontName=RLSmall-Italic
     outlineSettings {
      bApplyOutlineToDropShadows=false
      bMiteredCorners=false
      bSeparateFillAlpha=false
      outlineMaterial=null
      outlineSize=0
      outlineColor=<rgba 0 0 0 1>
     }
    }
   }
   slot {class=ButtonSlot, properties {horizontalAlignment=HAlign_Left, verticalAlignment=VAlign_Fill, padding=<ltrb 0 0 0 0>}}
  }
 ]
}
